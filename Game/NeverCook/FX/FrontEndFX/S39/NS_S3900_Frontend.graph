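# Niagara system for the Season 39 (S3900) frontend that sends occasional mesh birds flying across the scene.
# Two near-identical emitters probabilistically spawn birds on opposite Y sides, flying toward each other with randomized height and speed.
# A smoothed flap value and speed-scaled phase counter feed dynamic material parameters, apparently for material-driven wing flapping.
# The waveform-driven wander direction is smoothed but not visibly applied to velocity, and User.Speed is unreferenced in the visible modules.
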NiagaraSystem /Game/NeverCook/FX/FrontEndFX/S39/NS_S3900_Frontend
dependencies {usedDataInterfaces [&NiagaraDataInterfaceCurve], usedDynamicInputs [], usedModules [&SystemState]}
emitters {
 Birds {
  emitterUpdateScript [
   &EmitterState()
   &SpawnRate(Spawn Probability=0.2, SpawnRate=0.7, Use Spawn Probability=true) {module=SpawnRate001}
  ]
  particleSpawnScript [
   &InitializeParticle(
    Color=<rgba 0.7344 0.1839 0.1339 1>
    Lifetime=9
    Mesh Scale Mode=::NewEnumerator3
    Mesh Uniform Scale=0.3
    Sprite Size Mode=::NewEnumerator3
    Uniform Sprite Size=100
    UsePositionOffset=false
    Write Material Random=true
   )
   &ShapeLocation(
    Box Size=&RandomRangeVector(Minimum=<0 0 130>, Maximum=<1 1 200>)
    MaskSpawnGroup=false
    Offset=<0 700 400>
    Offset Mode=::NewEnumerator0
    Override Seed=false
    Shape Primitive=::NewEnumerator2
    Surface Only=false
    UseOwnerScale=true
   )
   &AddVelocity(
    Velocity=&RandomRangeVector(Minimum=<0 -1 -0.1>, Maximum=<0 -1 0.25>)
    Velocity Speed Scale=&RandomRangeFloat(Minimum=190, Maximum=210)
   )
  ]
  particleUpdateScript [
   &ParticleState()
   /SetVariables(Particles.RandomDirection:Vector3f=&Lerp_Vector(
     A=&MakeVector(
      X=&Waveform(
       Global Phase=&Add_Float(A=$Particles.Age, B=&RandomRangeFloat(Minimum=0, Maximum=5))
       "Waveform [1]"=::NewEnumerator8
       "[1] Frequency"=1
       "[1] Amplitude Min/Max"=<-1 1>
       "[2] Phase"=$Emitter.Age
       "[3] Phase"=$Emitter.Age
      )
      Y=&Waveform(
       Global Phase=&Add_Float(A=$Particles.Age, B=&RandomRangeFloat(Minimum=0, Maximum=5))
       "Waveform [1]"=::NewEnumerator8
       "[1] Frequency"=1
       "[1] Amplitude Min/Max"=<-1 1>
       "[2] Phase"=$Emitter.Age
       "[3] Phase"=$Emitter.Age
      )
      Z=0
     )
     B=<1 0 0>
     Alpha=&Waveform(
      Global Phase=&Add_Float(A=$Particles.Age, B=&RandomRangeFloat(Minimum=0, Maximum=5))
      Global Clamp Min=0.4
      "Waveform [1]"=::NewEnumerator8
      "[1] Frequency"=1
      "[1] Amplitude Min/Max"=<0 1>
      "[2] Phase"=$Emitter.Age
      "[3] Phase"=$Emitter.Age
     )
    )) {isSetParameters=true, module=SetVariables_#1}
   &InterpolateOverTime(
    Rate Of Change=0.5
    Target Vector Value:Vector3f=&NormalizeVector(Vector To Normalize=$Particles.RandomDirection, Safe Normalize=true, Safe Fallback Threshold=0, Safe Fallback Vector=<0 0 0.001>)
    Variable Type=::NewEnumerator2
   ) {module=InterpolateOverTime001}
   /SetVariables(Particles.FlapTarget=1) {isSetParameters=true, module=SetVariables_#2}
   &InterpolateOverTime(
    Rate Of Change:NiagaraFloat=&Lerp_Float(A=1, B=1, Alpha=$Particles.FlapTarget)
    Target Float Value:NiagaraFloat=$Particles.FlapTarget
   )
   /SetVariables(Particles.FlapSmoothed:NiagaraFloat=$Particles.InterpolateOverTime.InterpolateOverTime_Float.Moving Average) {isSetParameters=true, module=SetVariables_#3}
   /Niagara/Modules/Update/Utility/IncrementOverTime(Rate of Change:NiagaraFloat=&Multiply_Float(
     A=&Add_Float(A=0.15, B=$Particles.FlapSmoothed)
     B=&RemapRange(
      Input Value To Remap=&VectorLength(Vector=$Particles.Velocity)
      Input Min=0
      Input Max=5000
      Output Min=1
      Output Max=5
     )
    ))
   &SolveForcesAndVelocity(Clamp Velocity=true, Limit Acceleration=false, Speed Limit=800)
   &DynamicMaterialParameters(
    Index 0 Param 1=&Multiply_Float(A=$Particles.FlapSmoothed, B=1)
    Index 0 Param 2=&Multiply_Float(A=$Particles.IncrementOverTime.Counter, B=1)
    Module.Param0WriteEnabled=true
    Module.Param1WriteEnabled=true
    Module.Param2WriteEnabled=false
    Module.Param3WriteEnabled=false
   ) {module=DynamicMaterialParameters001}
  ]
  rendererClasses [&NiagaraMeshRendererProperties]
  renderers [{rendererClass=&NiagaraMeshRendererProperties, rendererIndex=0}]
 }
 Birds001 {
  emitterUpdateScript [
   &EmitterState()
   &SpawnRate(Spawn Probability=0.1, Use Spawn Probability=true) {module=SpawnRate001}
  ]
  particleSpawnScript [
   &InitializeParticle(
    Color=<rgba 0.7344 0.1839 0.1339 1>
    Lifetime=9
    Mesh Scale Mode=::NewEnumerator3
    Mesh Uniform Scale=0.3
    Sprite Size Mode=::NewEnumerator3
    Uniform Sprite Size=100
    UsePositionOffset=false
    Write Material Random=true
   )
   &ShapeLocation(
    Box Size=&RandomRangeVector(Minimum=<0 0 -200>, Maximum=<1 1 200>)
    MaskSpawnGroup=false
    Offset=<0 -700 300>
    Offset Mode=::NewEnumerator0
    Override Seed=false
    Shape Primitive=::NewEnumerator2
    Surface Only=false
    UseOwnerScale=true
   )
   &AddVelocity(
    Velocity=&RandomRangeVector(Minimum=<0 1 0.05>, Maximum=<0 1 0.25>)
    Velocity Speed Scale=&RandomRangeFloat(Minimum=190, Maximum=210)
   )
  ]
  particleUpdateScript [
   &ParticleState()
   /SetVariables(Particles.RandomDirection:Vector3f=&Lerp_Vector(
     A=&MakeVector(
      X=&Waveform(
       Global Phase=&Add_Float(A=$Particles.Age, B=&RandomRangeFloat(Minimum=0, Maximum=5))
       "Waveform [1]"=::NewEnumerator8
       "[1] Frequency"=1
       "[1] Amplitude Min/Max"=<-1 1>
       "[2] Phase"=$Emitter.Age
       "[3] Phase"=$Emitter.Age
      )
      Y=&Waveform(
       Global Phase=&Add_Float(A=$Particles.Age, B=&RandomRangeFloat(Minimum=0, Maximum=5))
       "Waveform [1]"=::NewEnumerator8
       "[1] Frequency"=1
       "[1] Amplitude Min/Max"=<-1 1>
       "[2] Phase"=$Emitter.Age
       "[3] Phase"=$Emitter.Age
      )
      Z=0
     )
     B=<1 0 0>
     Alpha=&Waveform(
      Global Phase=&Add_Float(A=$Particles.Age, B=&RandomRangeFloat(Minimum=0, Maximum=5))
      Global Clamp Min=0.4
      "Waveform [1]"=::NewEnumerator8
      "[1] Frequency"=1
      "[1] Amplitude Min/Max"=<0 1>
      "[2] Phase"=$Emitter.Age
      "[3] Phase"=$Emitter.Age
     )
    )) {isSetParameters=true, module=SetVariables_#1}
   &InterpolateOverTime(
    Rate Of Change=0.5
    Target Vector Value:Vector3f=&NormalizeVector(Vector To Normalize=$Particles.RandomDirection, Safe Normalize=true, Safe Fallback Threshold=0, Safe Fallback Vector=<0 0 0.001>)
    Variable Type=::NewEnumerator2
   ) {module=InterpolateOverTime001}
   /SetVariables(Particles.FlapTarget=1) {isSetParameters=true, module=SetVariables_#2}
   &InterpolateOverTime(
    Rate Of Change:NiagaraFloat=&Lerp_Float(A=1, B=1, Alpha=$Particles.FlapTarget)
    Target Float Value:NiagaraFloat=$Particles.FlapTarget
   )
   /SetVariables(Particles.FlapSmoothed:NiagaraFloat=$Particles.InterpolateOverTime.InterpolateOverTime_Float.Moving Average) {isSetParameters=true, module=SetVariables_#3}
   &IncrementOverTime(Rate of Change:NiagaraFloat=&Multiply_Float(
     A=&Add_Float(A=0.15, B=$Particles.FlapSmoothed)
     B=&RemapRange(
      Input Value To Remap=&VectorLength(Vector=$Particles.Velocity)
      Input Min=0
      Input Max=5000
      Output Min=1
      Output Max=5
     )
    ))
   &SolveForcesAndVelocity(Clamp Velocity=true, Limit Acceleration=false, Speed Limit=800)
   &DynamicMaterialParameters(
    Index 0 Param 1=&Multiply_Float(A=$Particles.FlapSmoothed, B=1)
    Index 0 Param 2=&Multiply_Float(A=$Particles.IncrementOverTime.Counter, B=1)
    Module.Param0WriteEnabled=true
    Module.Param1WriteEnabled=true
    Module.Param2WriteEnabled=false
    Module.Param3WriteEnabled=false
   ) {module=DynamicMaterialParameters001}
  ]
  rendererClasses [&NiagaraMeshRendererProperties]
  renderers [{rendererClass=&NiagaraMeshRendererProperties, rendererIndex=0}]
 }
}
summary {
 systemName=NS_S3900_Frontend
 emitters [
  {emitterName=Birds001, rendererClasses [&NiagaraMeshRendererProperties]}
  {emitterName=Birds, rendererClasses [&NiagaraMeshRendererProperties]}
 ]
 userVariables [User.Speed {type=NiagaraFloat, defaultValue=0, description=""}]
}
